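# Looping spot ambience cue for the Rufus Zero Point energy at the Asteria SkyIsland POI.
# Mixes two looping activity beds, a distant high-end and a close low-end, with magnetic-field discharge one-shots that repeat randomly after random delays.
# Each layer has its own attenuation asset, so the blend shifts with distance.
# A quality-level root plays all three layers on the first two levels and drops the discharges on the last two.
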SoundCue /Game/Sounds/Asteria/POI/SkyIsland/Zeropoint_Rufus/Spot_Zeropoint_Rufus_Energy_Loop_Cue
root=QualityLevel_0
nodes [
 {
  id=Attenuation_0
  children [Looping_0]
  properties {attenuationSettings=/Game/Sounds/Asteria/POI/Zeropoint_Petrified/Spots/ZP_Close_Pertrified_Att}
 }
 {
  id=Attenuation_1
  children [WavePlayer_17]
  properties {attenuationSettings=&/ZP_Rufus_Medium_Distant_Rufus_Att}
 }
 {
  id=Attenuation_2
  children [WavePlayer_18]
  properties {attenuationSettings=&/ZP_Rufus_Very_Close_Att}
 }
 {id=Delay_0, children [Random_0], properties {delayMax=8, delayMin=3}}
 {id=Looping_0, children [Delay_0]}
 {id=Mixer_1, children [Attenuation_1, Attenuation_0, Attenuation_2], properties {inputVolume [1.2,0.8,2]}}
 {id=Mixer_2, children [Attenuation_1, Attenuation_2], properties {inputVolume [1,1]}}
 {id=QualityLevel_0, children [Mixer_1,Mixer_1,Mixer_2,Mixer_2]}
 {
  id=Random_0
  children [WavePlayer_3, WavePlayer_11, WavePlayer_12, WavePlayer_13, WavePlayer_14, WavePlayer_15, WavePlayer_16]
  properties {weights [1,1,1,1,1,1,1]}
 }
 {id=WavePlayer_11, properties {soundWaveAssetPtr=&/ZP_MagneticField_Discharge_Swt_02}}
 {id=WavePlayer_12, properties {soundWaveAssetPtr=&/ZP_MagneticField_Discharge_Swt_03}}
 {id=WavePlayer_13, properties {soundWaveAssetPtr=&/ZP_MagneticField_Discharge_Swt_04}}
 {id=WavePlayer_14, properties {soundWaveAssetPtr=&/ZP_MagneticField_Discharge_Swt_05}}
 {id=WavePlayer_15, properties {soundWaveAssetPtr=&/ZP_MagneticField_Discharge_Swt_06}}
 {id=WavePlayer_16, properties {soundWaveAssetPtr=&/ZP_MagneticField_Discharge_Swt_07}}
 {
  id=WavePlayer_17
  properties {bLooping=true, soundWaveAssetPtr=&/ZP_Medium_Activity_Hi_End_Distant}
 }
 {
  id=WavePlayer_18
  properties {bLooping=true, soundWaveAssetPtr=&/ZP_Medium_Activity_Low_End_Close}
 }
 {id=WavePlayer_3, properties {soundWaveAssetPtr=&/ZP_MagneticField_Discharge_Swt_01}}
]
properties {
 attenuationSettings=&ZP_Close_Pertrified_Att
 maxDistance=6000
 soundClassObject=&Fort_Gadgets
 volumeMultiplier=0.6
 attenuationOverrides {
  bAttenuateWithLPF=true
  bEnableFocusInterpolation=true
  bEnableListenerFocus=true
  distanceAlgorithm=Custom
  falloffDistance=5500
  focusAttackInterpSpeed=0.5
  focusAzimuth=15
  focusDistanceScale=0.5
  focusReleaseInterpSpeed=0.25
  lPFFrequencyAtMax=500
  lPFRadiusMax=2000
  lPFRadiusMin=1000
  nonFocusAzimuth=90
  nonFocusVolumeAttenuation=0.1
  stereoSpread=2048
  attenuationShapeExtents {x=1024}
  customAttenuationCurve {editorCurveData {keys [<at 0 1>,<at 0.3 1>,<at 0.4 0.5>,<at 1 0>]}}
 }
}
